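# UI sound cue for clicking the friends list in the front end, routed to the Fort_UI sound class.
# A mixer layers three waves: a UI sound at full level, a pitched-up BigMenu sound, and a quiet Important_Update accent.
# The BigMenu layer runs through a fixed-pitch modulator and an enveloper that fades it out.
# A concurrency override prevents new instances once the limit is reached.
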SoundCue /Figment_S10/Sounds/Fort_UI_Sounds/FrontEnd/Fort_UI_FriendsListClicked_Cue
root=Mixer_1
nodes [
 {
  id=Enveloper_0
  children [Modulator_4]
  properties {pitchCurve {editorCurveData {keys [{time=0}]}}, volumeCurve {editorCurveData {keys [<at 0 1>,<at 0.3961 0.9966>,<at 1.7015 -0.0288>]}}}
 }
 {id=Mixer_1, children [WavePlayer_7, Enveloper_0, WavePlayer_32], properties {inputVolume [1,1,0.1]}}
 {id=Modulator_4, children [WavePlayer_17], properties {pitchMax=1.2, pitchMin=1.2}}
 {id=WavePlayer_17, properties {soundWaveAssetPtr=&/fort_ui_BigMenu_04}}
 {id=WavePlayer_32, properties {soundWaveAssetPtr=/Figment_S10/Sounds/Fort_UI_Sounds/Gameplay/fort_UI_Important_Update}}
 {id=WavePlayer_7, properties {soundWaveAssetPtr=&/fort_ui_new_chris_10}}
]
properties {
 bOverrideConcurrency=true
 maxDistance=2097152
 soundClassObject=&Fort_UI
 volumeMultiplier=0.7
 concurrencyOverrides {resolutionRule=PreventNew}
}
